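# Physics asset for the SlimyTune backpack cosmetic, simulating a dangling belt on the pack mesh.
# Small uniform capsules sit on a base belt bone and five dynamic belt bones chained beneath it.
# Each link constrains to the previous one with twist locked and varying, asymmetric swing limits, so the belt bends mostly along one swing axis.
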PhysicsAsset /BRCosmetics/Accessories/FORT_Backpacks/Backpack_F_MED_SlimyTune/Meshes/F_MED_SlimyTune_Pack_Physics
bodies {
 Base_belt [{shapeType=Capsule, length=3, radius=3, shapeName=Base_belt_capsule}]
 Dyn_belt_01 [{shapeType=Capsule, length=3, radius=3, shapeName=Dyn_belt_01_capsule}]
 Dyn_belt_02 [{shapeType=Capsule, length=3, radius=3, shapeName=Dyn_belt_02_capsule}]
 Dyn_belt_03 [{shapeType=Capsule, length=3, radius=3, shapeName=Dyn_belt_03_capsule}]
 Dyn_belt_04 [{shapeType=Capsule, length=3, radius=3, shapeName=Dyn_belt_04_capsule}]
 Dyn_belt_05 [{shapeType=Capsule, length=3, radius=3, shapeName=Dyn_belt_05_capsule}]
}
constraints [
 {bone1Name=Dyn_belt_01, bone2Name=Base_belt, swing1LimitDegrees=35, swing2LimitDegrees=5, twistLimitDegrees=0}
 {bone1Name=Dyn_belt_02, bone2Name=Dyn_belt_01, swing1LimitDegrees=30, swing2LimitDegrees=15, twistLimitDegrees=0}
 {bone1Name=Dyn_belt_03, bone2Name=Dyn_belt_02, swing1LimitDegrees=25, swing2LimitDegrees=10, twistLimitDegrees=0}
 {bone1Name=Dyn_belt_04, bone2Name=Dyn_belt_03, swing1LimitDegrees=20, swing2LimitDegrees=5, twistLimitDegrees=0}
 {bone1Name=Dyn_belt_05, bone2Name=Dyn_belt_04, swing1LimitDegrees=40, swing2LimitDegrees=0, twistLimitDegrees=0}
]
