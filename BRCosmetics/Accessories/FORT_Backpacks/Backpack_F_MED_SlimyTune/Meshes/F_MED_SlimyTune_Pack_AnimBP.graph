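# Animation blueprint for the SlimyTune backpack cosmetic, positioning the pack and adding secondary physics motion.
# Chained additive bone-space offsets adjust Base_Backpack and Base_belt, then a RigidBody node simulates relative to the pack bone with gravity override.
# Update logic computes state-based gravity, caches the parent's pelvis transform and derives a parachuting value from spine yaw.
# Customization load flags the F_MED_Prime mesh; the graphs show no reads of these stored values.
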AnimBlueprint /BRCosmetics/Accessories/FORT_Backpacks/Backpack_F_MED_SlimyTune/Meshes/F_MED_SlimyTune_Pack_AnimBP
target_skeleton=&/F_MED_SlimyTune_Pack_Skeleton
graphs [
 AnimGraph {nodes [
   n1 = "Transform (Modify) Bone - Bone: Base_Backpack"(ComponentPose=Local To Component(LocalPose=Input Pose()), Alpha=1, Translation="-2.2,0,0", Rotation="0,-5,0", Scale="1,1,1") {
    rotationMode=BMM_Additive
    rotationSpace=BCS_BoneSpace
    scaleMode=BMM_Additive
    scaleSpace=BCS_BoneSpace
    translationMode=BMM_Additive
    translationSpace=BCS_BoneSpace
    alphaBoolBlend {blendOption=Linear}
    boneToModify {boneName=Base_Backpack}
    rotation {yaw=-5}
    translation {x=-2.2}
   }
   n2 = "Transform (Modify) Bone - Bone: Base_Backpack"(ComponentPose=@n1, Alpha=1, Translation="4,0,6", Rotation="0,0,15", Scale="1,1,1") {
    rotationMode=BMM_Additive
    rotationSpace=BCS_BoneSpace
    scaleMode=BMM_Additive
    scaleSpace=BCS_BoneSpace
    translationMode=BMM_Additive
    translationSpace=BCS_BoneSpace
    alphaBoolBlend {blendOption=Linear}
    boneToModify {boneName=Base_Backpack}
    rotation {roll=15}
    translation {x=4, z=6}
   }
   n3 = "Transform (Modify) Bone - Bone: Base_belt"(ComponentPose=@n2, Alpha=1, Translation="0,0,0", Rotation="0,-80,0", Scale="1,1,1") {
    rotationMode=BMM_Additive
    rotationSpace=BCS_BoneSpace
    scaleMode=BMM_Additive
    scaleSpace=BCS_BoneSpace
    translationMode=BMM_Additive
    translationSpace=BCS_BoneSpace
    alphaBoolBlend {blendOption=Linear}
    boneToModify {boneName=Base_belt}
    rotation {yaw=-80}
   }
   n4 = "Transform (Modify) Bone - Bone: Base_Backpack"!(ComponentPose=@n3, bAlphaBoolEnabled:bool=True, Translation="-1.5,0,0", Rotation="0,0,0", Scale="1,1,1") -> (Pose) {
    alphaInputType=Bool
    rotationMode=BMM_Additive
    rotationSpace=BCS_BoneSpace
    scaleMode=BMM_Additive
    scaleSpace=BCS_BoneSpace
    translationMode=BMM_Additive
    translationSpace=BCS_BoneSpace
    alphaBoolBlend {blendOption=Linear}
    boneToModify {boneName=Base_Backpack}
    translation {x=-1.5}
   }
   n5 = RigidBody!(ComponentPose=@n4, Alpha, OverrideWorldGravity:Vector="0,0,-980", ExternalForce) -> (Pose) {
    bOverrideWorldGravity=true
    overridePhysicsAsset=&/F_MED_SlimyTune_Pack_Physics
    simulationSpace=BaseBoneSpace
    alphaBoolBlend {blendOption=Linear}
    alphaScaleBiasClamp {bClampResult=true}
    baseBoneRef {boneName=pack}
    componentAppliedLinearAccClamp=<3000 3000 2500>
    componentLinearAccScale=<0.25 0.2 0.5>
    componentLinearVelScale=<0.25 0.2 0.5>
    simSpaceSettings {maxAngularAcceleration=130, maxAngularVelocity=3, maxLinearAcceleration=130, maxLinearVelocity=150, velocityScaleZ=0, worldAlpha=1}
   }
   Output Pose(Result=Component To Local(ComponentPose=@n5)) {layerGroup=None}
  ]}
 EventGraph {nodes [
   Event BlueprintUpdateAnimation()
   > n1 = Sequence()
   > n2 = Gravity Override State Function DEPRECATED(
    CharacterPartAnimInstance=Self-Reference()
    GameTimeInSeconds=GetGameTimeInSeconds()
    GravityOverrideParams=Get Gravity Override Params()
    DebugJointName=dyn_hackBottomArmor
   )
   > Set Override World Gravity(Override World Gravity:Vector=@n2) -> (Output_Get:Vector)
   n3 = GetPartSkeletalMeshComponent()
   Is Valid(exec=@n1.then_1, InputObject=@n3)
   n4 = GetAttachParent!(self:Scene Component Object Reference=@n3) -> (ReturnValue:Scene Component Object Reference)
   > Is Valid(InputObject=@n4)
   n5 = GetSocketTransform!(self=@n4, InSocketName=pelvis, TransformSpace=RTS_World) -> (ReturnValue_Location:Vector, ReturnValue_Rotation:Rotator, ReturnValue_Scale:Vector)
   > Set PelvisLoc(PelvisLoc:Vector=@n5) -> (Output_Get:Vector)
   > Set PelvisRot(PelvisRot:Rotator=@n5.ReturnValue_Rotation) -> (Output_Get:Rotator)
   > Set PelvisScale(PelvisScale:Vector=@n5.ReturnValue_Scale) -> (Output_Get:Vector)
   n6 = GetDeltaTransformFromRefPose!(self=GetOwnerSkeletalMeshComponent(), BoneName=spine_05, BaseName=spine_01) -> (ReturnValue_Location:Vector, ReturnValue_Rotation:Rotator, ReturnValue_Scale:Vector)
   > Set Parashuting(Parashuting:double=MapRangeClamped(Value=BreakRotator(InRot=@n6.ReturnValue_Rotation).Yaw, InRangeA=-25, InRangeB=-65)) -> (Output_Get:double)
   n7 = Event OnCharacterCustomizationLoaded()
   n8 = Get Array Item with EFort Custom Part Type(Array:Array of Custom Character Part Object References=@n7.CustomParts) -> (Output:Custom Character Part Object Reference)
   > Set Character Prime(Character Prime:bool="Equal (Object)"(
     A=GetSkeletalMesh!(self:Custom Character Part Object Reference=@n8) -> (ReturnValue:Skeletal Mesh Object Reference)
     B=/DefaultCosmeticsAthena/Characters/Meshes/F_MED_Prime
    )) -> (Output_Get:bool)
  ]}
]
variables [
 Character Prime {type=boolean}
 Gravity Override Params {
  type=GravityOverrideParamsStruct
  default {
   dBNO {jointName=root, gravityOverride=<0 0 -980>}
   skydive_Parachuting {gravityOverride=<0 -800 -1500>, windAmplitude=<200 -200 100>, windFrequency=<1.5 4 4>}
   tactical_Sprint {gravityOverride=<0 -450 -400>}
  }
 }
 Override World Gravity {type=Vector}
 Parashuting {type=number}
 PelvisLoc {type=Vector}
 PelvisRot {type=Rotator}
 PelvisScale {type=Vector}
]
